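# Parent material for BurrBaby quadruped cosmetic companion accessories, adding a banner decal and optional metal finish to the shared character material.
# Banner_Texture is sampled twice with separate mirror, offset, rotation and scale; the summed red channels mask a base color multiply and roughness override.
# MetalBlend blends toward tinted, fully metallic attributes with roughness remapped into a min-max range.
# These edits use material attributes and named reroutes between the MF_CharacterParent_2023 PreFX and PostFX functions.
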Material /CosmeticCompanions/Assets/Quadruped/BurrBaby/Materials/M_BurrBaby_Accessory_Parent
graph {
 properties {
  bAutomaticallySetUsageInEditor=false
  bUseMaterialAttributes=true
  bUsedWithMeshDeformer=true
  bUsedWithSkeletalMesh=true
  bUsedWithStaticLighting=true
  shadingModel=MSM_FromMaterialExpression
  usedShadingModels=Default Lit | Subsurface | Clear Coat | Subsurface Profile | Cloth
 }
 nodes [
  n1 = MaterialFunctionCall() {materialFunction=&MF_CharacterParent_2023_PreFX}
  MaterialFunctionCall(Material=NamedRerouteUsage(), M=@n1.M, Pre Skinned Local Position=@n1.M, Distance Blend=@n1.M, Fresnel=@n1.M) {materialFunction=&MF_CharacterParent_2023_PostFX}
  n2 = ScalarParameter() {defaultValue=1, group=BurrBaby, parameterName=Banner1_MirrorX, sliderMax=1, sliderMin=-1}
  n3 = ScalarParameter() {defaultValue=1, group=BurrBaby, parameterName=Banner1_MirrorY, sliderMax=1, sliderMin=-1}
  n4 = Add(
   A=Multiply(A=AppendVector(A=@n2, B=@n3), B=TextureCoordinate())
   B=AppendVector(
    A=ScalarParameter() {defaultValue=0.338, group=BurrBaby, parameterName=Banner1_X}
    B=ScalarParameter() {defaultValue=0.27, group=BurrBaby, parameterName=Banner1_Y}
   )
  )
  n5 = ScalarParameter() {defaultValue=0.5, group=BurrBaby, parameterName=Banner1_Rotation}
  n6 = TextureSampleParameter2D(UVs=MaterialFunctionCall(
    UVs=MaterialFunctionCall(UVs=@n4, "Rotation Angle (0-1)"=@n5) {materialFunction=&CustomRotator}.Rotated Values
    Texture Scale=ScalarParameter() {defaultValue=0.23, group=BurrBaby, parameterName=Banner1_Scale}
   ) {materialFunction=&ScaleUVsByCenter}.UVs) {
   group=BurrBaby
   parameterName=Banner_Texture
   samplerSource=SSM_Clamp_WorldGroupSettings
   sortPriority=2
   texture=/CosmeticCompanions/Assets/Quadruped/BurrBaby/Textures/T_Banner_Default
  }
  n7 = ScalarParameter() {defaultValue=1, group=BurrBaby, parameterName=Banner2_MirrorX, sliderMax=1, sliderMin=-1}
  n8 = ScalarParameter() {defaultValue=1, group=BurrBaby, parameterName=Banner2_MirrorY, sliderMax=1, sliderMin=-1}
  n9 = Add(
   A=Multiply(A=AppendVector(A=@n7, B=@n8), B=TextureCoordinate())
   B=AppendVector(
    A=ScalarParameter() {defaultValue=0.338, group=BurrBaby, parameterName=Banner2_X}
    B=ScalarParameter() {defaultValue=-0.2737, group=BurrBaby, parameterName=Banner2_Y}
   )
  )
  n10 = ScalarParameter() {group=BurrBaby, parameterName=Banner2_Rotation}
  n11 = TextureSampleParameter2D(UVs=MaterialFunctionCall(
    UVs=MaterialFunctionCall(UVs=@n9, "Rotation Angle (0-1)"=@n10) {materialFunction=&CustomRotator}.Rotated Values
    Texture Scale=ScalarParameter() {defaultValue=0.23, group=BurrBaby, parameterName=Banner2_Scale}
   ) {materialFunction=&ScaleUVsByCenter}.UVs) {
   group=BurrBaby
   parameterName=Banner_Texture
   samplerSource=SSM_Clamp_WorldGroupSettings
   sortPriority=2
   texture=&T_Banner_Default
  }
  NamedRerouteDeclaration(None=Saturate(None=Add(A=@n6.R, B=@n11.R))) {name=BannerSample, nodeColor=<rgba 1 0 0.4219 1>}
  n12 = GetMaterialAttributes(None=NamedRerouteUsage()) {attributeGetTypes [#1, #2]}
  NamedRerouteDeclaration(None=@n12.Base Color) {name=BaseColor_IN, nodeColor=<rgba 1 0 0.0937 1>}
  n13 = NamedRerouteUsage()
  n14 = ScalarParameter() {defaultValue=0.7, group=BurrBaby, parameterName=BannerColorMultiply}
  NamedRerouteDeclaration(None=LinearInterpolate(A=@n13, B=Multiply(A=@n13, B=@n14), Alpha=NamedRerouteUsage())) {name=BaseColor_OUT, nodeColor=<rgba 0 0.4141 1 1>}
  NamedRerouteDeclaration(None=@n1.Result) {name=Material_IN, nodeColor=<rgba 0 0.5781 1 1>}
  n15 = NamedRerouteUsage()
  n16 = VectorParameter() {group=BurrBaby, parameterName=MetalColor, defaultValue=<rgba 0.5 0.5 0.5 1>}
  n17 = Constant() {r=1}
  n18 = NamedRerouteUsage()
  n19 = LinearInterpolate(
   A=ScalarParameter() {defaultValue=0.225, group=BurrBaby, parameterName=MetalRoughnessMin}
   B=ScalarParameter() {defaultValue=0.55, group=BurrBaby, parameterName=MetalRoughnessMax}
   Alpha=Saturate(None=GetMaterialAttributes(None=@n18) {attributeGetTypes [#2]}.Roughness)
  )
  NamedRerouteDeclaration(None=BlendMaterialAttributes(
    A=NamedRerouteUsage()
    B=SetMaterialAttributes!(MaterialAttributes=@n15, Base Color=@n16.RGB, Metallic=@n17, Roughness=@n19) {attributeSetTypes [#1, #3, #2]}
    Alpha=ScalarParameter() {group=BurrBaby, parameterName=MetalBlend}
   ) {vertexAttributeBlendType=UseA}) {name=Material_Metal, nodeColor=<rgba 0.8984 1 0 1>}
  NamedRerouteDeclaration(None=@n12.Roughness) {name=Roughness_IN, nodeColor=<rgba 1 0 0.0234 1>}
  NamedRerouteDeclaration(None=LinearInterpolate(A=NamedRerouteUsage(), B=ScalarParameter() {defaultValue=0.6, group=BurrBaby, parameterName=BannerRoughness}, Alpha=NamedRerouteUsage())) {name=RoughnessOUT, nodeColor=<rgba 0 0.3437 1 1>}
  NamedRerouteDeclaration(None=SetMaterialAttributes!(MaterialAttributes=NamedRerouteUsage(), Base Color=NamedRerouteUsage(), Roughness=NamedRerouteUsage()) {attributeSetTypes [#1, #2]}) {name=Material_OUT, nodeColor=<rgba 0.125 0 1 1>}
 ]
}
