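# Parent material for the BurrBaby quadruped cosmetic companion, built for skeletal meshes and mesh deformers.
# It chains shared material functions: character PreFX feeds companion custom color, then companion flipbook, then character PostFX.
# PreFX also passes mask, pre-skinned position, distance blend and fresnel data directly to PostFX.
# Material attributes carry the result, with the shading model chosen by expression.
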Material /CosmeticCompanions/Assets/Quadruped/BurrBaby/Materials/M_BurrBaby_Parent
graph {
 properties {
  bAutomaticallySetUsageInEditor=false
  bUseMaterialAttributes=true
  bUsedWithMeshDeformer=true
  bUsedWithSkeletalMesh=true
  bUsedWithStaticLighting=true
  shadingModel=MSM_FromMaterialExpression
  usedShadingModels=Default Lit | Subsurface | Clear Coat | Subsurface Profile | Cloth
 }
 nodes [
  n1 = MaterialFunctionCall() {materialFunction=&MF_CharacterParent_2023_PreFX}
  MaterialFunctionCall(
   Material=MaterialFunctionCall!(Material=MaterialFunctionCall!(Material=@n1.Result, Tech Art Mask=@n1.Result) {materialFunction=/CosmeticCompanions/Materials/MaterialFunctions/MF_FN_Companion_Custom_Color}.Base Color) {materialFunction=/CosmeticCompanions/Materials/MaterialFunctions/MF_FN_Companion_Flipbook}.Result
   M=@n1.M
   Pre Skinned Local Position=@n1.M
   Distance Blend=@n1.M
   Fresnel=@n1.M
  ) {materialFunction=&MF_CharacterParent_2023_PostFX}
 ]
}
