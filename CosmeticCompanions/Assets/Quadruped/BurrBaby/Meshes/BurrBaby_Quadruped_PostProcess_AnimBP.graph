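# Post-process AnimBP for the BurrBaby quadruped cosmetic companion, adding face poses, tail physics and eye attachment.
# The input pose runs through the BurrBaby face PoseAsset, then chained bone-relative AnimDynamics springs on tail_03 to tail_06.
# DynamicsAlpha drives tail weight through an inverting scale and bias; the tail_02 dynamics node is left unconnected.
# A final constraint binds the eyes bone to face_attach.
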AnimBlueprint /CosmeticCompanions/Assets/Quadruped/BurrBaby/Meshes/BurrBaby_Quadruped_PostProcess_AnimBP
target_skeleton=&/BurrBaby_Quadruped_Skeleton
graphs [
 AnimGraph {nodes [
   n1 = Get DynamicsAlpha() -> (DynamicsAlpha:float)
   Anim Dynamics - Bone: tail_02(Alpha=@n1, ExternalForce="0,0,0") {
    angularDampingOverride=1
    angularSpringConstant=15000
    bAngularSpring=true
    bOverrideAngularDamping=true
    bUseGravityOverride=true
    gravityScale=0.35
    simulationSpace=BoneRelative
    alphaBoolBlend {blendOption=Linear}
    alphaScaleBiasClamp {bias=1, scale=-1}
    boundBone {boneName=tail_02}
    componentLinearAccScale=<1 1 1>
    componentLinearVelScale=<1 1 1>
    gravityOverride {z=-980}
    physicsBodyDefinitions [{
      boundBone {boneName=tail_02}
      boxExtents=<20 30 30>
      constraintSetup {angularLimitsMax=<5 3 15>, angularLimitsMin=<-5 -45 -15>, angularTarget=<1 0 0>}
      localJointOffset=<4 0 0>
     }]
    relativeSpaceBone {boneName=root}
   }
   n2 = Input Pose()
   n3 = Anim Dynamics - Bone: tail_03(
    ComponentPose=Local To Component(LocalPose=BurrBaby_Face_PoseAsset!(SourcePose:Pose Link Structure=@n2) -> (Pose:Pose Link Structure) {
      is=title AnimGraphNode_PoseBlendNode
      blendOption=Linear
      poseAsset=&/BurrBaby_Face_PoseAsset
     })
    Alpha=@n1
    ExternalForce="0,0,0"
   ) {
    angularDampingOverride=1
    angularSpringConstant=15000
    bAngularSpring=true
    bOverrideAngularDamping=true
    bUseGravityOverride=true
    gravityScale=0.35
    simulationSpace=BoneRelative
    alphaBoolBlend {blendOption=Linear}
    alphaScaleBiasClamp {bias=1, scale=-1}
    boundBone {boneName=tail_03}
    componentLinearAccScale=<1 1 1>
    componentLinearVelScale=<1 1 1>
    gravityOverride {z=-980}
    physicsBodyDefinitions [{
      boundBone {boneName=tail_03}
      boxExtents=<20 30 30>
      constraintSetup {angularLimitsMax=<5 3 15>, angularLimitsMin=<-5 -45 -15>, angularTarget=<1 0 0>}
      localJointOffset=<4 0 0>
     }]
    relativeSpaceBone {boneName=root}
   }
   n4 = Anim Dynamics - Bone: tail_04(ComponentPose=@n3, Alpha=@n1, ExternalForce="0,0,0") {
    angularDampingOverride=1
    angularSpringConstant=15000
    bAngularSpring=true
    bOverrideAngularDamping=true
    bUseGravityOverride=true
    gravityScale=0.35
    simulationSpace=BoneRelative
    alphaBoolBlend {blendOption=Linear}
    alphaScaleBiasClamp {bias=1, scale=-1}
    boundBone {boneName=tail_04}
    componentLinearAccScale=<1 1 1>
    componentLinearVelScale=<1 1 1>
    gravityOverride {z=-980}
    physicsBodyDefinitions [{
      boundBone {boneName=tail_04}
      boxExtents=<20 30 30>
      constraintSetup {angularLimitsMax=<5 3 15>, angularLimitsMin=<-5 -45 -15>, angularTarget=<1 0 0>}
      localJointOffset=<4 0 0>
     }]
    relativeSpaceBone {boneName=root}
   }
   n5 = Anim Dynamics - Bone: tail_05(ComponentPose=@n4, Alpha=@n1, ExternalForce="0,0,0") {
    angularDampingOverride=1
    angularSpringConstant=15000
    bAngularSpring=true
    bOverrideAngularDamping=true
    bUseGravityOverride=true
    gravityScale=0.35
    simulationSpace=BoneRelative
    alphaBoolBlend {blendOption=Linear}
    alphaScaleBiasClamp {bias=1, scale=-1}
    boundBone {boneName=tail_05}
    componentLinearAccScale=<1 1 1>
    componentLinearVelScale=<1 1 1>
    gravityOverride {z=-980}
    physicsBodyDefinitions [{
      boundBone {boneName=tail_05}
      boxExtents=<20 30 30>
      constraintSetup {angularLimitsMax=<5 5 15>, angularLimitsMin=<-5 -45 -15>, angularTarget=<1 0 0>}
      localJointOffset=<4 0 0>
     }]
    relativeSpaceBone {boneName=root}
   }
   n6 = Anim Dynamics - Bone: tail_06(ComponentPose=@n5, Alpha=@n1, ExternalForce="0,0,0") {
    angularDampingOverride=1
    angularSpringConstant=15000
    bAngularSpring=true
    bOverrideAngularDamping=true
    bUseGravityOverride=true
    gravityScale=0.35
    simulationSpace=BoneRelative
    alphaBoolBlend {blendOption=Linear}
    alphaScaleBiasClamp {bias=1, scale=-1}
    boundBone {boneName=tail_06}
    componentLinearAccScale=<1 1 1>
    componentLinearVelScale=<1 1 1>
    gravityOverride {z=-980}
    physicsBodyDefinitions [{
      boundBone {boneName=tail_06}
      boxExtents=<20 30 30>
      constraintSetup {angularLimitsMax=<5 5 15>, angularLimitsMin=<-5 -45 -15>, angularTarget=<1 0 0>}
      localJointOffset=<4 0 0>
     }]
    relativeSpaceBone {boneName=root}
   }
   n7 = "Constraint - eyes (P)"(ComponentPose=@n6, Alpha=1, ConstraintWeights_0=1) {boneToModify {boneName=eyes}, constraintSetup [{targetBone {boneName=face_attach}}], constraintWeights [1]}
   Output Pose(Result=Component To Local(ComponentPose=@n7))
  ]}
 EventGraph {links [], nodes []}
]
variables []
